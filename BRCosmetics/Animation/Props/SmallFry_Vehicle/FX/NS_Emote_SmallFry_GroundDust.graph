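# Niagara ground-dust effect for the SmallFry vehicle emote, built from five sprite emitters.
# GroundRocks flings colliding pebbles, Smoke and Streaks001 push dust radially outward, and SmokeGroundRings grows upward-facing rings that partly inherit source movement.
# Those four zero their spawn rates when User.bIsInWater is true, while smoke_pulse adds a faster smoke burst without that check.
# Curves drive size, fades and dynamic material parameters.
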NiagaraSystem /BRCosmetics/Animation/Props/SmallFry_Vehicle/FX/NS_Emote_SmallFry_GroundDust
dependencies {
 usedDataInterfaces [&NiagaraDataInterfaceCurve, &NiagaraDataInterfaceColorCurve]
 usedDynamicInputs []
 usedModules [&SystemState]
}
emitters {
 GroundRocks {
  emitterUpdateScript [
   &EmitterState()
   &SpawnRate(
    SpawnRate=&Lerp_Float(
     A=0
     B=&RandomRangeFloat(Minimum=-15, Maximum=15, Evaluation Type=::NewEnumerator1)
     Alpha=&MakeCustomFloatFromBool(Bool=$User.bIsInWater, True Float=0, False Float=1)
    )
    Use Spawn Probability=false
   )
  ]
  particleSpawnScript [
   /SetVariables(Particles.MaterialRandom:NiagaraFloat=&RandomRangeFloat(Minimum=0, Maximum=1)) {isSetParameters=true}
   /Niagara/Modules/Spawn/Initialization/InitializeParticle(
    Color=&LerpLinearColors(StartColor=<rgba 0.6 0.495 0.42 1>, EndColor=<rgba 0.6 0.495 0.42 1>, LerpFactor=$Particles.MaterialRandom)
    Lifetime=&Lerp_Float(A=0.4, B=1, Alpha=$Particles.MaterialRandom)
    Module.Write Color=true
    Module.Write Lifetime=true
    Module.Write Mass=true
    Module.Write Material Random=false
    Module.Write Position=false
    Module.Write Scale=false
    Module.Write SpriteRotation=true
    Module.Write SpriteSize=true
    Position=&SimulationPosition()
    Sprite Rotation=&RandomRangeFloat(Minimum=0, Maximum=360)
    Sprite Size=&Vector2DFromFloat(Value=&RandomRangeFloat(Minimum=3, Maximum=7))
   )
   &SphereLocation(
    MaskSpawnGroup=false
    Non Uniform Scale=<1 1 0>
    Offset=<0 0 5>
    Override Seed=false
    Sphere Radius=&Lerp_Float(A=10, B=50, Alpha=&Power(a=$Particles.MaterialRandom, b=1.5))
    Surface Only=false
    UseNonUniformScale=true
    UseOffset=true
    UseSphereOrientation=false
   )
   &AddVelocityFromPoint(
    Module.Use Falloff Distance=false
    Velocity Strength=&Lerp_Float(A=150, B=400, Alpha=&Power(a=$Particles.MaterialRandom, b=3))
   )
   &AddVelocity(Velocity=<0 0 200>)
  ]
  particleUpdateScript [
   &ParticleState()
   &GravityForce()
   &Drag(Use Linear Drag=true, Use Rotational Drag=false)
   &SpriteRotationRate(Rotation Rate=&Lerp_Float(A=100, B=650, Alpha=$Particles.MaterialRandom))
   &ScaleSpriteSize(Scale Factor=&Vector2DFromFloat(Value=&FloatFromCurve(FloatCurve=DataInterface() {propertyValues {Curve {keys [
          {arriveTangent=35.6763, interpMode=RCIM_Cubic, leaveTangent=3.7835, tangentMode=RCTM_User, time=0, value=0}
          <at 0.7 1 RCIM_Cubic>
          {interpMode=RCIM_Cubic, leaveTangent=0.6903, time=1, value=0}
         ]}}})))
   &Collision(
    Advanced Aging Rate=1
    Age Colliding Particles=false
    Module.Control Rest State Via Surface Normal=false
    Module.EnableMaxCollisionCount=false
    Module.Randomize Collision Normal Vector=false
    Restitution=0.3
   )
   /SetVariables(Particles.MaxCollisions:NiagaraBool=&SetBoolByIntComparison(A=$Particles.Number Of Collisions, B=1, Comparison Type=::NewEnumerator0)) {isSetParameters=true}
   &SolveForcesAndVelocity(Clamp Velocity=false, Limit Acceleration=false)
   /SetVariables(Particles.SpriteRotation:NiagaraFloat=&MakeCustomFloatFromBool(Bool=$Particles.MaxCollisions, True Float=0, False Float=$Particles.SpriteRotation)) {isSetParameters=true}
  ]
  rendererClasses [&NiagaraSpriteRendererProperties]
  renderers [{rendererClass=&NiagaraSpriteRendererProperties}]
 }
 Smoke {
  emitterUpdateScript [
   &EmitterState()
   &SpawnRate(
    SpawnRate=&Lerp_Float(A=0, B=20, Alpha=&MakeCustomFloatFromBool(Bool=$User.bIsInWater, True Float=0, False Float=1))
    Use Spawn Probability=false
   )
  ]
  particleSpawnScript [
   /Niagara/Modules/Spawn/Initialization/V2/InitializeParticle(
    Color=&MakeLinearColorFromVectorAndFloat(
     "Vector (RGB)"=&Multiply_Vector(
      A=&MakeVectorFromLinearColorRGB(Color=<rgba 0.2344 0.2067 0.142 1>)
      B=&VectorFromFloat(Value=1)
     )
     "Float (Alpha)"=0.25
    )
    Lifetime=&RandomRangeFloat(Minimum=0.7, Maximum=1.1)
    Sprite Size Mode=::NewEnumerator1
    Sprite UV Mode=::NewEnumerator3
    Uniform Sprite Size Max=120
    Uniform Sprite Size Min=90
    UsePositionOffset=false
    Write Material Random=true
   )
   &CylinderLocation(
    Cylinder Height=0
    Cylinder Radius=&Lerp_Float(A=20, B=20, Alpha=$Particles.MaterialRandom)
    MaskSpawnGroup=false
    Module.Enable Lathe Profile=false
    Module.Override Randomness=false
    Module.Surface Only=true
    Surface Only Band Thickness=0
    Use Endcaps In Surface Only Mode=false
   )
   &AddVelocityFromPoint(Module.Use Falloff Distance=false, Velocity Strength=&Lerp_Float(A=50, B=150, Alpha=$Particles.MaterialRandom))
  ]
  particleUpdateScript [
   &ParticleState()
   &SolveForcesAndVelocity(Clamp Velocity=false, Limit Acceleration=false)
   &ScaleColor(
    Scale Alpha=&FloatFromCurve(FloatCurve=DataInterface() {propertyValues {Curve {keys [
         <at 0 0 RCIM_Cubic>
         {arriveTangent=-1.9555, interpMode=RCIM_Cubic, leaveTangent=-1.9555, tangentMode=RCTM_User, time=0.2374, value=1}
         <at 1 0>
        ]}}})
    ScaleA=true
    ScaleRGB=true
   )
   &ScaleSpriteSize(Scale Factor=&Vector2DFromFloat(Value=&FloatFromCurve(
      FloatCurve=DataInterface() {propertyValues {Curve {keys [<at 0 0.25 RCIM_Cubic>, {arriveTangent=1.1298, interpMode=RCIM_Cubic, leaveTangent=-0.5927, tangentMode=RCTM_User, time=1, value=1}]}}}
      Scale Curve=2
     )))
   &DynamicMaterialParameters(
    Index 0 Param 1=&FloatFromCurve(FloatCurve=DataInterface() {propertyValues {Curve {keys [<at 0 1>,<at 1 0>]}}})
    Module.Param0WriteEnabled=true
    Module.Param1WriteEnabled=true
    Module.Param2WriteEnabled=true
    Module.Param3WriteEnabled=true
   )
  ]
  rendererClasses [&NiagaraSpriteRendererProperties]
  renderers [{rendererClass=&NiagaraSpriteRendererProperties}]
 }
 SmokeGroundRings {
  emitterUpdateScript [
   &EmitterState()
   &SpawnRate(
    SpawnRate=&Lerp_Float(A=0, B=4, Alpha=&MakeCustomFloatFromBool(Bool=$User.bIsInWater, True Float=0, False Float=1))
    Use Spawn Probability=false
   )
  ]
  particleSpawnScript [
   /Niagara/Modules/Spawn/Initialization/InitializeParticle(
    Color=<rgba 0.35 0.2778 0.21 0.7>
    Lifetime=&RandomRangeFloat(Minimum=0.95, Maximum=1.1)
    Module.Write Color=true
    Module.Write Lifetime=true
    Module.Write Mass=true
    Module.Write Material Random=true
    Module.Write Position=false
    Module.Write Scale=false
    Module.Write SpriteRotation=true
    Module.Write SpriteSize=true
    Position=&SimulationPosition()
    Sprite Rotation=&UniformRangedFloat(Minimum=-180, Maximum=180)
    Sprite Size=&Vector2DFromFloat(Value=5)
   )
   /SetVariables(Particles.SpriteFacing=<0 0 1>) {isSetParameters=true}
   &SystemLocation(MaskSpawnGroup=false, Offset=<0 0 5>)
  ]
  particleUpdateScript [
   &ParticleState()
   &DynamicMaterialParameters(
    Index 0 Param 1=&FloatFromCurve(FloatCurve=DataInterface() {propertyValues {Curve {keys [<at 0 0.5>,<at 1 2>]}}})
    Module.Param0WriteEnabled=true
    Module.Param1WriteEnabled=false
    Module.Param2WriteEnabled=false
    Module.Param3WriteEnabled=false
   )
   &ScaleSpriteSize(Scale Factor=&Vector2DFromFloat(Value=&FloatFromCurve(
      FloatCurve=DataInterface() {propertyValues {Curve {keys [
          {arriveTangent=35.6763, interpMode=RCIM_Cubic, leaveTangent=3.7835, tangentMode=RCTM_User, time=0, value=0}
          {interpMode=RCIM_Cubic, leaveTangent=0.6903, time=1.0011, value=1}
         ]}}}
      Scale Curve=50
     )))
   &CameraOffset(Camera Offset Amount=20)
   &ScaleColor(
    Scale Alpha=&FloatFromCurve(FloatCurve=DataInterface() {propertyValues {Curve {keys [<at 0 1 RCIM_Cubic>, {interpMode=RCIM_Cubic, leaveTangent=-4.9981, time=1.0011, value=0.002}]}}})
    ScaleA=true
    ScaleRGB=false
   )
   &InheritSourceMovement(Applied Position Scale=<0.5 0.5 0.5>, ApplyForce=false, ApplyPosition=true, ApplyPositionFalloff=false)
   &SolveForcesAndVelocity(Clamp Velocity=false, Limit Acceleration=false)
  ]
  rendererClasses [&NiagaraSpriteRendererProperties]
  renderers [{rendererClass=&NiagaraSpriteRendererProperties}]
 }
 Streaks001 {
  emitterUpdateScript [
   &EmitterState()
   &SpawnRate(
    SpawnRate=&Lerp_Float(A=0, B=50, Alpha=&MakeCustomFloatFromBool(Bool=$User.bIsInWater, True Float=0, False Float=1))
    Use Spawn Probability=false
   )
  ]
  particleSpawnScript [
   &InitializeParticle(
    Color=&MakeLinearColorFromVectorAndFloat("Vector (RGB)"=&MakeVectorFromLinearColorRGB(Color=<rgba 1 0.8663 0.7274 1>), "Float (Alpha)"=0.125)
    Lifetime=&UniformRangedFloat(Minimum=0.3, Maximum=0.5)
    Module.Write Color=true
    Module.Write Lifetime=true
    Module.Write Mass=true
    Module.Write Material Random=true
    Module.Write Position=false
    Module.Write Scale=false
    Module.Write SpriteRotation=true
    Module.Write SpriteSize=true
    Position=&SimulationPosition()
    Sprite Size=&Multiply_Vector2(
     A=&UniformRangedVector2D(Minimum=<20 80>, Maximum=<40 160>)
     B=&Vector2DFromFloat(Value=0.3)
    )
   )
   &CylinderLocation(
    Cylinder Height=0
    Cylinder Radius=1
    MaskSpawnGroup=false
    Module.Enable Lathe Profile=false
    Module.Override Randomness=false
    Module.Surface Only=true
    Surface Only Band Thickness=1
    Use Endcaps In Surface Only Mode=false
   )
   &AddVelocity(Velocity=&MakeVector(X=0, Y=0, Z=&UniformRangedFloat(Minimum=0, Maximum=100)))
   &AddVelocityFromPoint(Module.Use Falloff Distance=false, Velocity Origin:Vector3f=&SimulationPosition(), Velocity Strength=300)
   &ApplyInitialForces()
   &DynamicMaterialParameters(
    Index 0 Param 1=&UniformRangedFloat(Minimum=0, Maximum=1)
    Index 0 Param 2=&UniformRangedFloat(Minimum=1, Maximum=2)
    Index 0 Param 3=1
    Index 0 Param 4=1
    Module.Param0WriteEnabled=true
    Module.Param1WriteEnabled=true
    Module.Param2WriteEnabled=true
    Module.Param3WriteEnabled=true
   )
  ]
  particleUpdateScript [
   &ParticleState()
   &Drag(Drag=1.5, Use Linear Drag=true, Use Rotational Drag=false)
   &SolveForcesAndVelocity(Clamp Velocity=false, Limit Acceleration=false)
   &ScaleColor(
    Scale Alpha=&FloatFromCurve(FloatCurve=DataInterface() {propertyValues {Curve {keys [<at 0 0 RCIM_Cubic>, {arriveTangent=0.002, interpMode=RCIM_Cubic, leaveTangent=0.002, time=0.5, value=1}, <at 1.0011 0.002 RCIM_Cubic>]}}})
    Scale RGB=&MakeVectorFromLinearColorRGB(Color=&ColorFromCurve(ColorCurve=DataInterface() {propertyValues {
        AlphaCurve {keys [<at 0 1 RCIM_Cubic>,<at 1 0 RCIM_Cubic>]}
        BlueCurve {keys [<at 0 1.14 RCIM_Cubic>,<at 1 1 RCIM_Cubic>]}
        GreenCurve {keys [<at 0 1.842 RCIM_Cubic>,<at 1 1 RCIM_Cubic>]}
        RedCurve {keys [<at 0 3 RCIM_Cubic>,<at 1 1 RCIM_Cubic>]}
       }}))
    ScaleA=true
    ScaleRGB=true
   )
  ]
  rendererClasses [&NiagaraSpriteRendererProperties]
  renderers [{rendererClass=&NiagaraSpriteRendererProperties}]
 }
 smoke_pulse {
  emitterUpdateScript [
   &EmitterState(Life Cycle Mode=::NewEnumerator1, Loop Duration=0.2, UseLoopDelay=false)
   &SpawnBurst_Instantaneous(Spawn Count=15, Use Spawn Probability=false)
  ]
  particleSpawnScript [
   /Niagara/Modules/Spawn/Initialization/V2/InitializeParticle(
    Color=&MakeLinearColorFromVectorAndFloat(
     "Vector (RGB)"=&Multiply_Vector(
      A=&MakeVectorFromLinearColorRGB(Color=<rgba 0.2344 0.2067 0.142 1>)
      B=&VectorFromFloat(Value=1)
     )
     "Float (Alpha)"=0.25
    )
    Lifetime=&RandomRangeFloat(Minimum=0.7, Maximum=1.1)
    Sprite Size Mode=::NewEnumerator1
    Sprite UV Mode=::NewEnumerator3
    Uniform Sprite Size Max=120
    Uniform Sprite Size Min=90
    UsePositionOffset=false
    Write Material Random=true
   )
   &CylinderLocation(
    Cylinder Height=0
    Cylinder Radius=&Lerp_Float(A=20, B=20, Alpha=$Particles.MaterialRandom)
    MaskSpawnGroup=false
    Module.Enable Lathe Profile=false
    Module.Override Randomness=false
    Module.Surface Only=true
    Surface Only Band Thickness=0
    Use Endcaps In Surface Only Mode=false
   )
   &AddVelocityFromPoint(Module.Use Falloff Distance=false, Velocity Strength=&Lerp_Float(A=250, B=300, Alpha=$Particles.MaterialRandom))
  ]
  particleUpdateScript [
   &ParticleState()
   &SolveForcesAndVelocity(Clamp Velocity=false, Limit Acceleration=false)
   &ScaleColor(
    Scale Alpha=&FloatFromCurve(FloatCurve=DataInterface() {propertyValues {Curve {keys [
         <at 0 0 RCIM_Cubic>
         {arriveTangent=-1.9555, interpMode=RCIM_Cubic, leaveTangent=-1.9555, tangentMode=RCTM_User, time=0.2374, value=1}
         <at 1 0>
        ]}}})
    ScaleA=true
    ScaleRGB=true
   )
   &ScaleSpriteSize(Scale Factor=&Vector2DFromFloat(Value=&FloatFromCurve(FloatCurve=DataInterface() {propertyValues {Curve {keys [<at 0 0.5 RCIM_Cubic>, {arriveTangent=0.5532, interpMode=RCIM_Cubic, leaveTangent=-1.4995, tangentMode=RCTM_User, time=1, value=1}]}}})))
   &DynamicMaterialParameters(
    Index 0 Param 1=&FloatFromCurve(FloatCurve=DataInterface() {propertyValues {Curve {keys [<at 0 1>,<at 1 0>]}}})
    Module.Param0WriteEnabled=true
    Module.Param1WriteEnabled=true
    Module.Param2WriteEnabled=true
    Module.Param3WriteEnabled=true
   )
  ]
  rendererClasses [&NiagaraSpriteRendererProperties]
  renderers [{rendererClass=&NiagaraSpriteRendererProperties}]
 }
}
summary {
 systemName=NS_Emote_SmallFry_GroundDust
 emitters [
  {emitterName=SmokeGroundRings, rendererClasses [&NiagaraSpriteRendererProperties]}
  {emitterName=Streaks001, rendererClasses [&NiagaraSpriteRendererProperties]}
  {emitterName=Smoke, rendererClasses [&NiagaraSpriteRendererProperties]}
  {emitterName=GroundRocks, rendererClasses [&NiagaraSpriteRendererProperties]}
  {emitterName=smoke_pulse, rendererClasses [&NiagaraSpriteRendererProperties]}
 ]
 userVariables [User.bIsInWater {type=NiagaraBool}]
}
