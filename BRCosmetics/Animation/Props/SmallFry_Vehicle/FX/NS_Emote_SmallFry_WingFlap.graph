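# Niagara system for the SmallFry vehicle emote that renders blurred, rapidly flapping wings as translucent mesh particles.
# Four emitters form two layers per side, placed each update on User.SkeletalMesh via SkeletalMeshLocation with offsets.
# Continuously spawned short-lived particles take their flap angle from a sawtooth of emitter age, so overlapping meshes fan out.
# The A layers mirror shared System offset and twist values, while the smaller, fainter B layers flap slower.
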NiagaraSystem /BRCosmetics/Animation/Props/SmallFry_Vehicle/FX/NS_Emote_SmallFry_WingFlap
dependencies {
 usedDataInterfaces []
 usedDynamicInputs []
 usedModules [
  &self:SystemSpawnScript.SystemScriptSource.SystemScriptGraph.NiagaraNodeAssignment_0.SetVariables_#1
  &SystemState
 ]
}
emitters {
 Left_Wing_A {
  emitterUpdateScript [
   &EmitterState()
   &SpawnRate(SpawnRate=80, Use Spawn Probability=false)
  ]
  particleSpawnScript [
   /SetVariables(Particles.Random01:NiagaraFloat=&RandomRangeFloat(Minimum=0, Maximum=1)) {isSetParameters=true, module=SetVariables_#2}
   /SetVariables() {isSetParameters=true, module=SetVariables_#3}
   /SetVariables(Transient.WingFlapRangeTopRange=50) {isSetParameters=true, module=SetVariables_#4}
   /SetVariables(Transient.WingFlapRangeBottomRange=-5) {isSetParameters=true, module=SetVariables_#5}
   &InitializeParticle(
    Color=&MakeLinearColorFromVectorAndFloat("Vector (RGB)"=&MakeVectorFromLinearColorRGB(Color=<rgba 0.75 0.65 0.55 1>), "Float (Alpha)"=0.2)
    Lifetime=0.1
    Mesh Scale=<2.5 5 1>
    Mesh Scale Mode=::NewEnumerator4
    Sprite Size Mode=::NewEnumerator3
    Uniform Sprite Size=20
    UsePositionOffset=false
    Write Material Random=true
   )
   &InitialMeshOrientation(Rotation=$System.WingATwist, Use Orientation Vector=true, Use Rotation Vector=true)
   /SetVariables(Particles.MeshOrientation=&MultiplyQuaternion(
     Quaternion A=$Particles.MeshOrientation
     Quaternion B=&MultiplyQuaternion(
      Quaternion A=&MakeQuatFromAxisAngle(
       Axis=<1 0 0>
       AngleInDegrees=&Lerp_Float(A=0, B=0, Alpha=&Frac_Float(Float=&Multiply_Float(A=$Emitter.Age, B=10)))
      )
      Quaternion B=&MultiplyQuaternion(
       Quaternion A=&MakeQuatFromAxisAngle(
        Axis=<0 1 0>
        AngleInDegrees=&Lerp_Float(
         A=$Transient.WingFlapRangeBottomRange
         B=$Transient.WingFlapRangeTopRange
         Alpha=&Frac_Float(Float=&Multiply_Float(A=$Emitter.Age, B=15.5))
        )
       )
       Quaternion B=&MakeQuatFromAxisAngle(
        Axis=<0 0 1>
        AngleInDegrees=&Lerp_Float(A=0, B=0, Alpha=&Frac_Float(Float=&Multiply_Float(A=$Emitter.Age, B=10)))
       )
      )
     )
    )) {isSetParameters=true, module=SetVariables_#6}
  ]
  particleUpdateScript [
   &SkeletalMeshLocation(
    Inherit Velocity=false
    Interpolate Mesh Sampling=false
    MaskSpawnGroup=false
    Mesh Sampling Type=::NewEnumerator5
    Sampled Position Offset=$System.WingAOffset
    Sampled Position Offset Coordinate Space:ENiagaraMeshTransforms=::NewEnumerator3
    Skeletal Mesh=$User.SkeletalMesh
    Socket Sampling Mode=::NewEnumerator5
   )
   &ParticleState()
   &UpdateMeshOrientation(
    Rotation Rate=0
    Rotation Vector=&RotateVectorByQuaternion(
     Vector To Rotate=&ReturnMeshOrientationAxis(
      Mesh Orientation Axis=::XAxis
      Mesh Orientation Quat=&MultiplyQuaternion(
       Quaternion A=&OrientMeshToVector(Look At Direction=<1 0 -1>, Reference Axis Vector=<1 0 0>)
       Quaternion B=<0 0 0 1>
      )
     )
     Quaternion=<0 0 0 1>
    )
   ) {enabled=false}
   &CameraOffset(Camera Offset Amount=10) {enabled=false}
  ]
  rendererClasses [&NiagaraMeshRendererProperties]
  renderers [
   {rendererClass=&NiagaraMeshRendererProperties, rendererIndex=0}
   {rendererClass=&NiagaraMeshRendererProperties, rendererIndex=1}
  ]
 }
 Left_Wing_B001 {
  emitterUpdateScript [
   &EmitterState()
   &SpawnRate(SpawnRate=60, Use Spawn Probability=false)
  ]
  particleSpawnScript [
   /SetVariables(Particles.Random01:NiagaraFloat=&RandomRangeFloat(Minimum=0, Maximum=1)) {isSetParameters=true, module=SetVariables_#2}
   /SetVariables() {isSetParameters=true, module=SetVariables_#3}
   /SetVariables(Transient.WingFlapRangeTopRange=-50) {isSetParameters=true, module=SetVariables_#7}
   /SetVariables() {isSetParameters=true, module=SetVariables_#8}
   &InitializeParticle(
    Color=&MakeLinearColorFromVectorAndFloat("Vector (RGB)"=&MakeVectorFromLinearColorRGB(Color=<rgba 0.75 0.65 0.55 1>), "Float (Alpha)"=0.1)
    Lifetime=0.1
    Mesh Scale=<1.9 5 1>
    Mesh Scale Mode=::NewEnumerator4
    Sprite Size Mode=::NewEnumerator3
    Uniform Sprite Size=20
    UsePositionOffset=false
    Write Material Random=true
   )
   &InitialMeshOrientation(Rotation=<0 0.05 0>, Use Orientation Vector=true, Use Rotation Vector=true)
   /SetVariables(Particles.MeshOrientation=&MultiplyQuaternion(
     Quaternion A=$Particles.MeshOrientation
     Quaternion B=&MultiplyQuaternion(
      Quaternion A=&MakeQuatFromAxisAngle(
       Axis=<1 0 0>
       AngleInDegrees=&Lerp_Float(A=0, B=0, Alpha=&Frac_Float(Float=&Multiply_Float(A=$Emitter.Age, B=10)))
      )
      Quaternion B=&MultiplyQuaternion(
       Quaternion A=&MakeQuatFromAxisAngle(
        Axis=<0 1 0>
        AngleInDegrees=&Lerp_Float(
         A=$Transient.WingFlapRangeBottomRange
         B=$Transient.WingFlapRangeTopRange
         Alpha=&Frac_Float(Float=&Multiply_Float(A=$Emitter.Age, B=8.5))
        )
       )
       Quaternion B=&MakeQuatFromAxisAngle(
        Axis=<0 0 1>
        AngleInDegrees=&Lerp_Float(A=0, B=0, Alpha=&Frac_Float(Float=&Multiply_Float(A=$Emitter.Age, B=10)))
       )
      )
     )
    )) {isSetParameters=true, module=SetVariables_#6}
  ]
  particleUpdateScript [
   &SkeletalMeshLocation(
    Inherit Velocity=false
    Interpolate Mesh Sampling=false
    MaskSpawnGroup=false
    Mesh Sampling Type=::NewEnumerator5
    Sampled Position Offset=<-8 0 0>
    Sampled Position Offset Coordinate Space:ENiagaraMeshTransforms=::NewEnumerator3
    Skeletal Mesh=$User.SkeletalMesh
    Socket Sampling Mode=::NewEnumerator5
   )
   &ParticleState()
   &UpdateMeshOrientation(Rotation Rate=0, Rotation Vector=<1 0 0>) {enabled=false}
  ]
  rendererClasses [&NiagaraMeshRendererProperties]
  renderers [
   {rendererClass=&NiagaraMeshRendererProperties, rendererIndex=0}
   {rendererClass=&NiagaraMeshRendererProperties, rendererIndex=1}
  ]
 }
 Right_Wing_A {
  emitterUpdateScript [
   &EmitterState()
   &SpawnRate(SpawnRate=80, Use Spawn Probability=false)
  ]
  particleSpawnScript [
   /SetVariables(Particles.Random01:NiagaraFloat=&RandomRangeFloat(Minimum=0, Maximum=1)) {isSetParameters=true, module=SetVariables_#2}
   /SetVariables() {isSetParameters=true, module=SetVariables_#3}
   /SetVariables(Transient.WingFlapRangeTopRange=-50) {isSetParameters=true, module=SetVariables_#4}
   /SetVariables(Transient.WingFlapRangeBottomRange=5) {isSetParameters=true, module=SetVariables_#5}
   &InitializeParticle(
    Color=&MakeLinearColorFromVectorAndFloat("Vector (RGB)"=&MakeVectorFromLinearColorRGB(Color=<rgba 0.75 0.65 0.55 1>), "Float (Alpha)"=0.2)
    Lifetime=0.1
    Mesh Scale=<2.5 5 1>
    Mesh Scale Mode=::NewEnumerator4
    Sprite Size Mode=::NewEnumerator3
    Uniform Sprite Size=20
    UsePositionOffset=false
    Write Material Random=true
   )
   &InitialMeshOrientation(
    Orientation Axis=<0 0 -1>
    Orientation Vector=<0 0 1>
    Rotation=&Multiply_Vector(A=<-1 -1 1>, B=$System.WingATwist)
    Use Orientation Vector=true
    Use Rotation Vector=true
   )
   /SetVariables(Particles.MeshOrientation=&MultiplyQuaternion(
     Quaternion A=$Particles.MeshOrientation
     Quaternion B=&MultiplyQuaternion(
      Quaternion A=&MakeQuatFromAxisAngle(
       Axis=<1 0 0>
       AngleInDegrees=&Lerp_Float(A=0, B=0, Alpha=&Frac_Float(Float=&Multiply_Float(A=$Emitter.Age, B=10)))
      )
      Quaternion B=&MultiplyQuaternion(
       Quaternion A=&MakeQuatFromAxisAngle(
        Axis=<0 1 0>
        AngleInDegrees=&Lerp_Float(
         A=$Transient.WingFlapRangeBottomRange
         B=$Transient.WingFlapRangeTopRange
         Alpha=&Frac_Float(Float=&Multiply_Float(A=$Emitter.Age, B=15.5))
        )
       )
       Quaternion B=&MakeQuatFromAxisAngle(
        Axis=<0 0 1>
        AngleInDegrees=&Lerp_Float(A=0, B=0, Alpha=&Frac_Float(Float=&Multiply_Float(A=$Emitter.Age, B=10)))
       )
      )
     )
    )) {isSetParameters=true, module=SetVariables_#6}
  ]
  particleUpdateScript [
   &SkeletalMeshLocation(
    Bone / Socket Index=1
    Inherit Velocity=false
    Interpolate Mesh Sampling=false
    MaskSpawnGroup=false
    Mesh Sampling Type=::NewEnumerator5
    Sampled Position Offset=&Multiply_Vector(A=$System.WingAOffset, B=<1 -1 1>)
    Sampled Position Offset Coordinate Space:ENiagaraMeshTransforms=::NewEnumerator3
    Skeletal Mesh=$User.SkeletalMesh
    Socket Sampling Mode=::NewEnumerator5
   )
   &ParticleState()
   &UpdateMeshOrientation(
    Rotation Rate=0
    Rotation Vector=&RotateVectorByQuaternion(
     Vector To Rotate=&ReturnMeshOrientationAxis(
      Mesh Orientation Axis=::XAxis
      Mesh Orientation Quat=&MultiplyQuaternion(
       Quaternion A=&OrientMeshToVector(Look At Direction=<1 0 -1>, Reference Axis Vector=<1 0 0>)
       Quaternion B=<0 0 0 1>
      )
     )
     Quaternion=<0 0 0 1>
    )
   ) {enabled=false}
  ]
  rendererClasses [&NiagaraMeshRendererProperties]
  renderers [
   {rendererClass=&NiagaraMeshRendererProperties, rendererIndex=0}
   {rendererClass=&NiagaraMeshRendererProperties, rendererIndex=1}
  ]
 }
 Right_Wing_B {
  emitterUpdateScript [
   &EmitterState()
   &SpawnRate(SpawnRate=60, Use Spawn Probability=false)
  ]
  particleSpawnScript [
   /SetVariables(Particles.Random01:NiagaraFloat=&RandomRangeFloat(Minimum=0, Maximum=1)) {isSetParameters=true, module=SetVariables_#2}
   /SetVariables() {isSetParameters=true, module=SetVariables_#3}
   /SetVariables(Transient.WingFlapRangeTopRange=-45) {isSetParameters=true, module=SetVariables_#7}
   /SetVariables(Transient.WingFlapRangeBottomRange=10) {isSetParameters=true, module=SetVariables_#8}
   &InitializeParticle(
    Color=&MakeLinearColorFromVectorAndFloat("Vector (RGB)"=&MakeVectorFromLinearColorRGB(Color=<rgba 0.75 0.65 0.55 1>), "Float (Alpha)"=0.1)
    Lifetime=0.1
    Mesh Scale=<1.9 5 1>
    Mesh Scale Mode=::NewEnumerator4
    Sprite Size Mode=::NewEnumerator3
    Uniform Sprite Size=20
    UsePositionOffset=false
    Write Material Random=true
   )
   &InitialMeshOrientation(Orientation Axis=<0 0 1>, Orientation Vector=<0 0 -1>, Rotation=<0 0.05 0>, Use Orientation Vector=true, Use Rotation Vector=true)
   /SetVariables(Particles.MeshOrientation=&MultiplyQuaternion(
     Quaternion A=$Particles.MeshOrientation
     Quaternion B=&MultiplyQuaternion(
      Quaternion A=&MakeQuatFromAxisAngle(
       Axis=<1 0 0>
       AngleInDegrees=&Lerp_Float(A=0, B=0, Alpha=&Frac_Float(Float=&Multiply_Float(A=$Emitter.Age, B=10)))
      )
      Quaternion B=&MultiplyQuaternion(
       Quaternion A=&MakeQuatFromAxisAngle(
        Axis=<0 1 0>
        AngleInDegrees=&Lerp_Float(
         A=$Transient.WingFlapRangeBottomRange
         B=$Transient.WingFlapRangeTopRange
         Alpha=&Frac_Float(Float=&Multiply_Float(A=$Emitter.Age, B=8.5))
        )
       )
       Quaternion B=&MakeQuatFromAxisAngle(
        Axis=<0 0 1>
        AngleInDegrees=&Lerp_Float(A=0, B=0, Alpha=&Frac_Float(Float=&Multiply_Float(A=$Emitter.Age, B=10)))
       )
      )
     )
    )) {isSetParameters=true, module=SetVariables_#6}
  ]
  particleUpdateScript [
   &SkeletalMeshLocation(
    Bone / Socket Index=1
    Inherit Velocity=false
    Interpolate Mesh Sampling=false
    MaskSpawnGroup=false
    Mesh Sampling Type=::NewEnumerator5
    Sampled Position Offset=<-8 0 0>
    Sampled Position Offset Coordinate Space:ENiagaraMeshTransforms=::NewEnumerator3
    Skeletal Mesh=$User.SkeletalMesh
    Socket Sampling Mode=::NewEnumerator5
   )
   &ParticleState()
   &UpdateMeshOrientation(Rotation Rate=0, Rotation Vector=<1 0 0>) {enabled=false}
  ]
  rendererClasses [&NiagaraMeshRendererProperties]
  renderers [
   {rendererClass=&NiagaraMeshRendererProperties, rendererIndex=0}
   {rendererClass=&NiagaraMeshRendererProperties, rendererIndex=1}
  ]
 }
}
summary {
 systemName=NS_Emote_SmallFry_WingFlap
 emitters [
  {emitterName=Left_Wing_A, rendererClasses [&NiagaraMeshRendererProperties]}
  {emitterName=Left_Wing_B001, rendererClasses [&NiagaraMeshRendererProperties]}
  {emitterName=Right_Wing_B, rendererClasses [&NiagaraMeshRendererProperties]}
  {emitterName=Right_Wing_A, rendererClasses [&NiagaraMeshRendererProperties]}
 ]
 userVariables [
  User.PhysMatIndex {type=NiagaraInt32}
  User.SkeletalMesh {
   type=NiagaraDataInterfaceSkeletalMesh
   defaultValue {
    struct=NiagaraExt_VariableValue_DataInterface
    value {dataInterface=&self:NiagaraDataInterfaceSkeletalMesh_0}
   }
  }
 ]
}
